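# Parent material for the MuzzleSwish companion emote prop, a masked skeletal-mesh material with an animated energy reveal.
# It layers the effect between the character parent PreFX and PostFX functions, after companion custom color.
# Progress parameters, radial coordinates, panning noise and a vertex-color appear index build reveal masks, with temporally dithered opacity.
# BodyID and AccentID switch tint, accent, fresnel and outline colors, and time of day tints base and emissive.
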Material /CosmeticCompanions/Assets/Props/Emotes/MuzzleSwish/Materials/M_MuzzleSwish_Prop_Parent
graph {
 expressions {
  MaterialExpressionAdd_1 {
   class=MaterialExpressionAdd
   inputs [
    A {from=MaterialExpressionMultiply_3}
    B {from=MaterialExpressionAppendVector_0}
   ]
  }
  MaterialExpressionAdd_2 {
   class=MaterialExpressionAdd
   inputs [
    A {from=MaterialExpressionNamedRerouteUsage_9}
    B {from=MaterialExpressionLinearInterpolate_2}
   ]
  }
  MaterialExpressionAdd_3 {
   class=MaterialExpressionAdd
   inputs [
    A {from=MaterialExpressionPower_1}
    B {from=MaterialExpressionTextureSample_0, output=R}
   ]
  }
  MaterialExpressionAppendVector_0 {
   class=MaterialExpressionAppendVector
   inputs [
    A {from=MaterialExpressionConstant_0, output=""}
    B {from=MaterialExpressionNamedRerouteUsage_5, output=""}
   ]
  }
  MaterialExpressionAppendVector_1 {
   class=MaterialExpressionAppendVector
   inputs [
    A {from=MaterialExpressionConstant_0, output=""}
    B {from=MaterialExpressionScalarParameter_5, output=""}
   ]
  }
  MaterialExpressionComponentMask_0 {class=MaterialExpressionComponentMask, inputs [None {from=MaterialExpressionDivide_0}], properties {g=true, r=true}}
  MaterialExpressionConstant3Vector_0 {class=MaterialExpressionConstant3Vector, properties {constant=<rgba 1 1 1 0>}}
  MaterialExpressionConstant3Vector_1 {class=MaterialExpressionConstant3Vector}
  MaterialExpressionConstant_0 {class=MaterialExpressionConstant}
  MaterialExpressionConstant_10 {class=MaterialExpressionConstant, properties {r=0.7}}
  MaterialExpressionConstant_11 {class=MaterialExpressionConstant, properties {r=0.4}}
  MaterialExpressionConstant_12 {class=MaterialExpressionConstant, properties {r=0.8}}
  MaterialExpressionConstant_2 {class=MaterialExpressionConstant, properties {r=0.1}}
  MaterialExpressionConstant_3 {class=MaterialExpressionConstant, properties {r=0.5}}
  MaterialExpressionConstant_4 {class=MaterialExpressionConstant, properties {r=0.6}}
  MaterialExpressionConstant_5 {class=MaterialExpressionConstant}
  MaterialExpressionConstant_6 {class=MaterialExpressionConstant, properties {r=1}}
  MaterialExpressionConstant_7 {class=MaterialExpressionConstant, properties {r=0.9}}
  MaterialExpressionConstant_8 {class=MaterialExpressionConstant, properties {r=0.3}}
  MaterialExpressionConstant_9 {class=MaterialExpressionConstant, properties {r=0.2}}
  MaterialExpressionConvert_0 {
   class=MaterialExpressionConvert
   inputs [None {from=MaterialExpressionMaterialFunctionCall_3, output=Radial Coordinates}]
   properties {
    nodeName=Break Float2
    convertInputs [{
      type=Vector2
      expressionInput {
       expression=&self:MaterialExpressionMaterialFunctionCall_3
       inputName=None
       mask=0
       maskA=0
       maskB=0
       maskG=0
       maskR=0
       outputIndex=0
      }
     }]
    convertOutputs [{type=Scalar}, {type=Scalar}]
   }
  }
  MaterialExpressionDDX_0 {class=MaterialExpressionDDX, inputs [Value {from=MaterialExpressionTextureCoordinate_0, output=""}], properties {}}
  MaterialExpressionDDY_0 {class=MaterialExpressionDDY, inputs [Value {from=MaterialExpressionTextureCoordinate_0, output=""}], properties {}}
  MaterialExpressionDivide_0 {
   class=MaterialExpressionDivide
   inputs [
    A {from=MaterialExpressionNamedRerouteUsage_14}
    B {from=MaterialExpressionScalarParameter_2}
   ]
   properties {constB=256}
  }
  MaterialExpressionGetMaterialAttributes_0 {
   class=MaterialExpressionGetMaterialAttributes
   inputs [None {from=MaterialExpressionMaterialFunctionCall_2, output=Base Color}]
   properties {attributeGetTypes [
     #1
     #2
     #3
     #4
    ]}
  }
  MaterialExpressionInverseLinearInterpolate_0 {
   class=MaterialExpressionInverseLinearInterpolate
   inputs [
    A {from=MaterialExpressionLinearInterpolate_3, output=""}
    B {from="", output=""}
    Value {from=MaterialExpressionSaturate_0, output=""}
   ]
   properties {constA=0.25}
  }
  MaterialExpressionInverseLinearInterpolate_1 {
   class=MaterialExpressionInverseLinearInterpolate
   inputs [
    A {from="", output=""}
    B {from="", output=""}
    Value {from=MaterialExpressionNamedRerouteUsage_13, output=""}
   ]
   properties {constB=0.7}
  }
  MaterialExpressionInverseLinearInterpolate_2 {
   class=MaterialExpressionInverseLinearInterpolate
   inputs [
    A {from="", output=""}
    B {from=MaterialExpressionLinearInterpolate_11, output=""}
    Value {from=MaterialExpressionLinearInterpolate_12, output=""}
   ]
   properties {constA=0.1}
  }
  MaterialExpressionInverseLinearInterpolate_3 {
   class=MaterialExpressionInverseLinearInterpolate
   inputs [
    A {from=MaterialExpressionSubtract_11, output=""}
    B {from=MaterialExpressionSubtract_14, output=""}
    Value {from=MaterialExpressionNamedRerouteUsage_28, output=""}
   ]
   properties {}
  }
  MaterialExpressionInverseLinearInterpolate_4 {
   class=MaterialExpressionInverseLinearInterpolate
   inputs [
    A {from=MaterialExpressionLinearInterpolate_3, output=""}
    B {from="", output=""}
    Value {from=MaterialExpressionMultiply_1, output=""}
   ]
   properties {constA=0.25}
  }
  MaterialExpressionLinearInterpolate_0 {
   class=MaterialExpressionLinearInterpolate
   inputs [
    A {from=MaterialExpressionMultiply_22}
    B {from=""}
    Alpha {from=MaterialExpressionMultiply_11}
   ]
   properties {constB=0}
  }
  MaterialExpressionLinearInterpolate_1 {
   class=MaterialExpressionLinearInterpolate
   inputs [
    A {from=MaterialExpressionNamedRerouteUsage_2}
    B {from=MaterialExpressionNamedRerouteUsage_10}
    Alpha {from=MaterialExpressionMultiply_11}
   ]
  }
  MaterialExpressionLinearInterpolate_11 {
   class=MaterialExpressionLinearInterpolate
   inputs [
    A {from=""}
    B {from=""}
    Alpha {from=MaterialExpressionScalarParameter_19}
   ]
   properties {constA=1, constB=0.2}
  }
  MaterialExpressionLinearInterpolate_12 {
   class=MaterialExpressionLinearInterpolate
   inputs [
    A {from=MaterialExpressionMultiply_34}
    B {from=MaterialExpressionNamedRerouteUsage_32}
    Alpha {from=MaterialExpressionScalarParameter_20}
   ]
  }
  MaterialExpressionLinearInterpolate_2 {
   class=MaterialExpressionLinearInterpolate
   inputs [
    A {from=MaterialExpressionMultiply_9}
    B {from=MaterialExpressionConstant3Vector_1}
    Alpha {from=MaterialExpressionNamedRerouteUsage_22}
   ]
  }
  MaterialExpressionLinearInterpolate_3 {
   class=MaterialExpressionLinearInterpolate
   inputs [
    A {from=""}
    B {from=""}
    Alpha {from=MaterialExpressionSaturate_15}
   ]
   properties {constB=0.15}
  }
  MaterialExpressionLinearInterpolate_4 {
   class=MaterialExpressionLinearInterpolate
   inputs [
    A {from=MaterialExpressionLinearInterpolate_1}
    B {from=MaterialExpressionNamedRerouteUsage_7}
    Alpha {from=MaterialExpressionMultiply_21}
   ]
  }
  MaterialExpressionLinearInterpolate_6 {
   class=MaterialExpressionLinearInterpolate
   inputs [
    A {from=MaterialExpressionLinearInterpolate_0}
    B {from=MaterialExpressionNamedRerouteUsage_15}
    Alpha {from=MaterialExpressionSaturate_7}
   ]
  }
  MaterialExpressionMaterialFunctionCall_0 {
   class=MaterialExpressionMaterialFunctionCall
   inputs [
    Overide Texture Set UVs {from=""}
    Override Tech Art Mask UVs {from=""}
    Override Basic Emissive UVs {from=""}
   ]
   properties {materialFunction=&MF_CharacterParent_2023_PreFX}
  }
  MaterialExpressionMaterialFunctionCall_1 {
   class=MaterialExpressionMaterialFunctionCall
   inputs [
    Material {from=MaterialExpressionSetMaterialAttributes_0}
    M {from=MaterialExpressionMaterialFunctionCall_0, output=M}
    Pre Skinned Local Position {from=MaterialExpressionNamedRerouteDeclaration_14}
    Distance Blend {from=MaterialExpressionMaterialFunctionCall_0, output=M}
    Fresnel {from=MaterialExpressionNamedRerouteDeclaration_13}
   ]
   properties {materialFunction=&MF_CharacterParent_2023_PostFX}
  }
  MaterialExpressionMaterialFunctionCall_12 {
   class=MaterialExpressionMaterialFunctionCall
   inputs [
    Color_Morning {from=MaterialExpressionConstant3Vector_0}
    Color_Day {from=MaterialExpressionConstant3Vector_0}
    Color_Evening {from=MaterialExpressionConstant3Vector_0}
    Color_Night {from=MaterialExpressionVectorParameter_4, output=RGB}
   ]
   properties {materialFunction=&Color_Time_of_Day}
  }
  MaterialExpressionMaterialFunctionCall_2 {
   class=MaterialExpressionMaterialFunctionCall
   inputs [
    Material {from=MaterialExpressionMaterialFunctionCall_0, output=Result}
    Tech Art Mask {from=MaterialExpressionMaterialFunctionCall_0, output=Result}
   ]
   properties {materialFunction=/CosmeticCompanions/Materials/MaterialFunctions/MF_FN_Companion_Custom_Color}
  }
  MaterialExpressionMaterialFunctionCall_3 {
   class=MaterialExpressionMaterialFunctionCall
   inputs [
    Vector or UVs {from=MaterialExpressionComponentMask_0}
    Swizzle Coordinate Output {from=""}
   ]
   properties {materialFunction=&VectorToRadialValue}
  }
  MaterialExpressionMaterialFunctionCall_5 {
   class=MaterialExpressionMaterialFunctionCall
   inputs [
    Alpha Threshold {from=MaterialExpressionMaterialFunctionCall_6, output=Lerp 3 Inputs}
    Random {from=""}
   ]
   properties {materialFunction=/Game/Environments/Asteria/Sets/NeonCity/Techart/Screen/MaterialFunctions/DitherTemporalAABlue}
  }
  MaterialExpressionMaterialFunctionCall_6 {
   class=MaterialExpressionMaterialFunctionCall
   inputs [
    1 {from=MaterialExpressionConstant_5}
    2 {from=MaterialExpressionPower_0}
    3 {from=MaterialExpressionConstant_6}
    4 {from=""}
    A {from=MaterialExpressionNamedRerouteUsage_22}
   ]
   properties {materialFunction=/Engine/Functions/Engine_MaterialFunctions02/Lerp_Multiple_Float}
  }
  MaterialExpressionMaterialFunctionCall_8 {
   class=MaterialExpressionMaterialFunctionCall
   inputs [
    Base {from=MaterialExpressionMultiply_1}
    Add {from=MaterialExpressionPower_1}
   ]
   properties {materialFunction=/Engine/Functions/Engine_MaterialFunctions02/Math/MultiplyAdd}
  }
  MaterialExpressionMaterialFunctionCall_9 {
   class=MaterialExpressionMaterialFunctionCall
   inputs [
    Alpha Threshold {from=MaterialExpressionMaterialFunctionCall_6, output=Lerp 3 Inputs}
    Offset {from=""}
    Random {from=""}
    Offset2 {from=""}
   ]
   properties {materialFunction=&DitherTemporalAA_Fortnite}
  }
  MaterialExpressionMultiply_1 {
   class=MaterialExpressionMultiply
   inputs [
    A {from=MaterialExpressionConvert_0}
    B {from=MaterialExpressionScalarParameter_1}
   ]
   properties {constB=0.2}
  }
  MaterialExpressionMultiply_11 {
   class=MaterialExpressionMultiply
   inputs [
    A {from=MaterialExpressionMultiply_2}
    B {from=MaterialExpressionPower_0}
   ]
  }
  MaterialExpressionMultiply_14 {
   class=MaterialExpressionMultiply
   inputs [
    A {from=MaterialExpressionPower_2}
    B {from=MaterialExpressionNamedRerouteUsage_16}
   ]
  }
  MaterialExpressionMultiply_2 {
   class=MaterialExpressionMultiply
   inputs [
    A {from=MaterialExpressionSaturate_2}
    B {from=MaterialExpressionNamedRerouteDeclaration_20}
   ]
  }
  MaterialExpressionMultiply_20 {
   class=MaterialExpressionMultiply
   inputs [
    A {from=MaterialExpressionNamedRerouteUsage_1}
    B {from=MaterialExpressionNamedRerouteUsage_11}
   ]
  }
  MaterialExpressionMultiply_21 {
   class=MaterialExpressionMultiply
   inputs [
    A {from=MaterialExpressionPower_6}
    B {from=MaterialExpressionMultiply_14}
   ]
  }
  MaterialExpressionMultiply_22 {
   class=MaterialExpressionMultiply
   inputs [
    A {from=MaterialExpressionOneMinus_7}
    B {from=MaterialExpressionMultiply_20}
   ]
  }
  MaterialExpressionMultiply_3 {
   class=MaterialExpressionMultiply
   inputs [
    A {from=MaterialExpressionMaterialFunctionCall_3, output=Radial Coordinates}
    B {from=""}
   ]
   properties {constB=2}
  }
  MaterialExpressionMultiply_34 {
   class=MaterialExpressionMultiply
   inputs [
    A {from=MaterialExpressionNamedRerouteUsage_31}
    B {from=""}
   ]
   properties {constB=2}
  }
  MaterialExpressionMultiply_35 {
   class=MaterialExpressionMultiply
   inputs [
    A {from=MaterialExpressionLinearInterpolate_4}
    B {from=MaterialExpressionMaterialFunctionCall_12, output=Result}
   ]
  }
  MaterialExpressionMultiply_36 {
   class=MaterialExpressionMultiply
   inputs [
    A {from=MaterialExpressionMaterialFunctionCall_12, output=Result}
    B {from=MaterialExpressionLinearInterpolate_6}
   ]
  }
  MaterialExpressionMultiply_4 {
   class=MaterialExpressionMultiply
   inputs [
    A {from=MaterialExpressionVertexColor_1, output=R}
    B {from=""}
   ]
   properties {constB=10}
  }
  MaterialExpressionMultiply_5 {
   class=MaterialExpressionMultiply
   inputs [
    A {from=MaterialExpressionConvert_0}
    B {from=""}
   ]
   properties {constB=2}
  }
  MaterialExpressionMultiply_6 {
   class=MaterialExpressionMultiply
   inputs [
    A {from=MaterialExpressionNamedRerouteUsage_6}
    B {from=""}
   ]
   properties {constB=2}
  }
  MaterialExpressionMultiply_9 {
   class=MaterialExpressionMultiply
   inputs [
    A {from=MaterialExpressionVertexNormalWS_0}
    B {from=MaterialExpressionScalarParameter_6}
   ]
  }
  MaterialExpressionNamedRerouteDeclaration_0 {
   class=MaterialExpressionNamedRerouteDeclaration
   inputs [None {from=MaterialExpressionGetMaterialAttributes_0, output=Emissive Color}]
   properties {name=Emissive_IN, nodeColor=<rgba 0 0.625 1 1>}
  }
  MaterialExpressionNamedRerouteDeclaration_1 {
   class=MaterialExpressionNamedRerouteDeclaration
   inputs [None {from=MaterialExpressionMultiply_35}]
   properties {name=Emissive_OUT, nodeColor=<rgba 0.9453 1 0 1>}
  }
  MaterialExpressionNamedRerouteDeclaration_10 {
   class=MaterialExpressionNamedRerouteDeclaration
   inputs [None {from=MaterialExpressionGetMaterialAttributes_0, output=Opacity}]
   properties {name=Opacity_IN, nodeColor=<rgba 1 0.0938 0 1>}
  }
  MaterialExpressionNamedRerouteDeclaration_11 {
   class=MaterialExpressionNamedRerouteDeclaration
   inputs [None {from=MaterialExpressionSwitch_0}]
   properties {name=AppearIndex, nodeColor=<rgba 0 1 0.4844 1>}
  }
  MaterialExpressionNamedRerouteDeclaration_12 {
   class=MaterialExpressionNamedRerouteDeclaration
   inputs [None {from=MaterialExpressionScalarParameter_4}]
   properties {name=AlphaProgress, nodeColor=<rgba 0.9652 0 1 1>}
  }
  MaterialExpressionNamedRerouteDeclaration_13 {
   class=MaterialExpressionNamedRerouteDeclaration
   inputs [None {from=MaterialExpressionMaterialFunctionCall_0, output=Fresnel}]
   properties {name=Fresnel, nodeColor=<rgba 0 0.1328 1 1>}
  }
  MaterialExpressionNamedRerouteDeclaration_14 {
   class=MaterialExpressionNamedRerouteDeclaration
   inputs [None {from=MaterialExpressionMaterialFunctionCall_0, output=Pre skinned Local Position}]
   properties {name=PreskinnedPos, nodeColor=<rgba 1 0 0.9375 1>}
  }
  MaterialExpressionNamedRerouteDeclaration_15 {
   class=MaterialExpressionNamedRerouteDeclaration
   inputs [None {from=MaterialExpressionSaturate_16}]
   properties {name=Tips Mask, nodeColor=<rgba 0 0.6719 1 1>}
  }
  MaterialExpressionNamedRerouteDeclaration_16 {
   class=MaterialExpressionNamedRerouteDeclaration
   inputs [None {from=MaterialExpressionTextureSample_0, output=R}]
   properties {name=LFEnergyNoise, nodeColor=<rgba 0.6172 1 0 1>}
  }
  MaterialExpressionNamedRerouteDeclaration_17 {
   class=MaterialExpressionNamedRerouteDeclaration
   inputs [None {from=MaterialExpressionOneMinus_3}]
   properties {name=RevealMask, nodeColor=<rgba 0 1 0.2031 1>}
  }
  MaterialExpressionNamedRerouteDeclaration_18 {
   class=MaterialExpressionNamedRerouteDeclaration
   inputs [None {from=MaterialExpressionSwitch_4}]
   properties {name=OutlineColor, nodeColor=<rgba 0.0012 0.4849 1 1>}
  }
  MaterialExpressionNamedRerouteDeclaration_19 {
   class=MaterialExpressionNamedRerouteDeclaration
   inputs [None {from=MaterialExpressionScalarParameter_18}]
   properties {name=Progress_CeilBias, nodeColor=<rgba 1 0 0.6328 1>}
  }
  MaterialExpressionNamedRerouteDeclaration_2 {
   class=MaterialExpressionNamedRerouteDeclaration
   inputs [None {from=MaterialExpressionGetMaterialAttributes_0, output=Base Color}]
   properties {name=Basecolor_IN, nodeColor=<rgba 0 1 0.8125 1>}
  }
  MaterialExpressionNamedRerouteDeclaration_20 {
   class=MaterialExpressionNamedRerouteDeclaration
   inputs [None {from=MaterialExpressionSaturate_5}]
   properties {name=Tips Mask with Fresnel, nodeColor=<rgba 0 0.6719 1 1>}
  }
  MaterialExpressionNamedRerouteDeclaration_21 {
   class=MaterialExpressionNamedRerouteDeclaration
   inputs [None {from=MaterialExpressionSwitch_1}]
   properties {name=FresnelColor, nodeColor=<rgba 0.0012 0.4849 1 1>}
  }
  MaterialExpressionNamedRerouteDeclaration_22 {
   class=MaterialExpressionNamedRerouteDeclaration
   inputs [None {from=MaterialExpressionSwitch_3}]
   properties {name=BaseColorTint, nodeColor=<rgba 0.0012 0.4849 1 1>}
  }
  MaterialExpressionNamedRerouteDeclaration_23 {
   class=MaterialExpressionNamedRerouteDeclaration
   inputs [None {from=MaterialExpressionSwitch_2}]
   properties {name=BaseColorAccent, nodeColor=<rgba 0.0012 0.4849 1 1>}
  }
  MaterialExpressionNamedRerouteDeclaration_3 {
   class=MaterialExpressionNamedRerouteDeclaration
   inputs [None {from=MaterialExpressionLinearInterpolate_6}]
   properties {name=Basecolor_OUT, nodeColor=<rgba 1 0.4922 0 1>}
  }
  MaterialExpressionNamedRerouteDeclaration_4 {
   class=MaterialExpressionNamedRerouteDeclaration
   inputs [None {from=MaterialExpressionScalarParameter_3}]
   properties {name=Progress, nodeColor=<rgba 1 0 0.6328 1>}
  }
  MaterialExpressionNamedRerouteDeclaration_5 {
   class=MaterialExpressionNamedRerouteDeclaration
   inputs [None {from=MaterialExpressionSaturate_2}]
   properties {name=RevealRamp, nodeColor=<rgba 1 0 0.5391 1>}
  }
  MaterialExpressionNamedRerouteDeclaration_6 {
   class=MaterialExpressionNamedRerouteDeclaration
   inputs [None {from=MaterialExpressionMaterialFunctionCall_9, output=Result}]
   properties {name=OpacityMask_OUT, nodeColor=<rgba 0.1719 0 1 1>}
  }
  MaterialExpressionNamedRerouteDeclaration_7 {
   class=MaterialExpressionNamedRerouteDeclaration
   inputs [None {from=MaterialExpressionGetMaterialAttributes_0, output=World Position Offset}]
   properties {name=WPO_IN, nodeColor=<rgba 0.1953 0 1 1>}
  }
  MaterialExpressionNamedRerouteDeclaration_8 {
   class=MaterialExpressionNamedRerouteDeclaration
   inputs [None {from=MaterialExpressionNamedRerouteUsage_9}]
   properties {name=WPO_OUT, nodeColor=<rgba 1 0 0.4453 1>}
  }
  MaterialExpressionNamedRerouteDeclaration_9 {
   class=MaterialExpressionNamedRerouteDeclaration
   inputs [None {from=MaterialExpressionMaterialFunctionCall_6, output=Lerp 3 Inputs}]
   properties {name=Opacity_OUT, nodeColor=<rgba 1 0.2578 0 1>}
  }
  MaterialExpressionNamedRerouteUsage_0 {class=MaterialExpressionNamedRerouteUsage}
  MaterialExpressionNamedRerouteUsage_1 {class=MaterialExpressionNamedRerouteUsage}
  MaterialExpressionNamedRerouteUsage_10 {class=MaterialExpressionNamedRerouteUsage}
  MaterialExpressionNamedRerouteUsage_11 {class=MaterialExpressionNamedRerouteUsage}
  MaterialExpressionNamedRerouteUsage_12 {class=MaterialExpressionNamedRerouteUsage}
  MaterialExpressionNamedRerouteUsage_13 {class=MaterialExpressionNamedRerouteUsage}
  MaterialExpressionNamedRerouteUsage_14 {class=MaterialExpressionNamedRerouteUsage}
  MaterialExpressionNamedRerouteUsage_15 {class=MaterialExpressionNamedRerouteUsage}
  MaterialExpressionNamedRerouteUsage_16 {class=MaterialExpressionNamedRerouteUsage}
  MaterialExpressionNamedRerouteUsage_19 {class=MaterialExpressionNamedRerouteUsage}
  MaterialExpressionNamedRerouteUsage_2 {class=MaterialExpressionNamedRerouteUsage}
  MaterialExpressionNamedRerouteUsage_20 {class=MaterialExpressionNamedRerouteUsage}
  MaterialExpressionNamedRerouteUsage_22 {class=MaterialExpressionNamedRerouteUsage}
  MaterialExpressionNamedRerouteUsage_27 {class=MaterialExpressionNamedRerouteUsage}
  MaterialExpressionNamedRerouteUsage_28 {class=MaterialExpressionNamedRerouteUsage}
  MaterialExpressionNamedRerouteUsage_3 {class=MaterialExpressionNamedRerouteUsage}
  MaterialExpressionNamedRerouteUsage_30 {class=MaterialExpressionNamedRerouteUsage}
  MaterialExpressionNamedRerouteUsage_31 {class=MaterialExpressionNamedRerouteUsage}
  MaterialExpressionNamedRerouteUsage_32 {class=MaterialExpressionNamedRerouteUsage}
  MaterialExpressionNamedRerouteUsage_33 {class=MaterialExpressionNamedRerouteUsage}
  MaterialExpressionNamedRerouteUsage_4 {class=MaterialExpressionNamedRerouteUsage}
  MaterialExpressionNamedRerouteUsage_5 {class=MaterialExpressionNamedRerouteUsage}
  MaterialExpressionNamedRerouteUsage_6 {class=MaterialExpressionNamedRerouteUsage}
  MaterialExpressionNamedRerouteUsage_7 {class=MaterialExpressionNamedRerouteUsage}
  MaterialExpressionNamedRerouteUsage_8 {class=MaterialExpressionNamedRerouteUsage}
  MaterialExpressionNamedRerouteUsage_9 {class=MaterialExpressionNamedRerouteUsage}
  MaterialExpressionOneMinus_3 {class=MaterialExpressionOneMinus, inputs [None {from=MaterialExpressionSaturate_14}]}
  MaterialExpressionOneMinus_6 {class=MaterialExpressionOneMinus, inputs [None {from=MaterialExpressionNamedRerouteUsage_19}]}
  MaterialExpressionOneMinus_7 {class=MaterialExpressionOneMinus, inputs [None {from=MaterialExpressionPower_7}]}
  MaterialExpressionPanner_0 {
   class=MaterialExpressionPanner
   inputs [
    Coordinate {from=MaterialExpressionAdd_1}
    Time {from=""}
    Speed {from=MaterialExpressionAppendVector_1}
   ]
   properties {speedY=-0.5}
  }
  MaterialExpressionPower_0 {
   class=MaterialExpressionPower
   inputs [
    Base {from=MaterialExpressionSmoothStep_0}
    Exp {from=""}
   ]
  }
  MaterialExpressionPower_1 {
   class=MaterialExpressionPower
   inputs [
    Base {from=MaterialExpressionNamedRerouteUsage_12}
    Exp {from=MaterialExpressionScalarParameter_0}
   ]
  }
  MaterialExpressionPower_2 {
   class=MaterialExpressionPower
   inputs [
    Base {from=MaterialExpressionSaturate_6}
    Exp {from=MaterialExpressionScalarParameter_7}
   ]
  }
  MaterialExpressionPower_6 {
   class=MaterialExpressionPower
   inputs [
    Base {from=MaterialExpressionOneMinus_6}
    Exp {from=""}
   ]
   properties {constExponent=6}
  }
  MaterialExpressionPower_7 {
   class=MaterialExpressionPower
   inputs [
    Base {from=MaterialExpressionNamedRerouteUsage_20}
    Exp {from=""}
   ]
  }
  MaterialExpressionRound_0 {class=MaterialExpressionRound, inputs [None {from=MaterialExpressionMultiply_4, output=""}], properties {}}
  MaterialExpressionSaturate_0 {class=MaterialExpressionSaturate, inputs [None {from=MaterialExpressionMaterialFunctionCall_8, output=Result}]}
  MaterialExpressionSaturate_14 {class=MaterialExpressionSaturate, inputs [None {from=MaterialExpressionInverseLinearInterpolate_3}]}
  MaterialExpressionSaturate_15 {class=MaterialExpressionSaturate, inputs [None {from=MaterialExpressionNamedRerouteUsage_6}]}
  MaterialExpressionSaturate_16 {class=MaterialExpressionSaturate, inputs [None {from=MaterialExpressionInverseLinearInterpolate_4}]}
  MaterialExpressionSaturate_2 {class=MaterialExpressionSaturate, inputs [None {from=MaterialExpressionSubtract_1}]}
  MaterialExpressionSaturate_4 {class=MaterialExpressionSaturate, inputs [None {from=MaterialExpressionAdd_3}]}
  MaterialExpressionSaturate_5 {class=MaterialExpressionSaturate, inputs [None {from=MaterialExpressionInverseLinearInterpolate_0}]}
  MaterialExpressionSaturate_6 {class=MaterialExpressionSaturate, inputs [None {from=MaterialExpressionInverseLinearInterpolate_1}]}
  MaterialExpressionSaturate_7 {class=MaterialExpressionSaturate, inputs [None {from=MaterialExpressionInverseLinearInterpolate_2}]}
  MaterialExpressionScalarParameter_0 {class=MaterialExpressionScalarParameter, properties {defaultValue=1, group=MuzzleSwish, parameterName=MS_Exp}}
  MaterialExpressionScalarParameter_1 {class=MaterialExpressionScalarParameter, properties {defaultValue=1, group=MuzzleSwish, parameterName=MS_TailEndIntensity}}
  MaterialExpressionScalarParameter_10 {class=MaterialExpressionScalarParameter, properties {group=MuzzleSwish, parameterName=AccentID, sliderMax=2}}
  MaterialExpressionScalarParameter_11 {class=MaterialExpressionScalarParameter, properties {group=MuzzleSwish, parameterName=AccentID, sliderMax=2}}
  MaterialExpressionScalarParameter_12 {class=MaterialExpressionScalarParameter, properties {group=MuzzleSwish, parameterName=AccentID, sliderMax=2}}
  MaterialExpressionScalarParameter_18 {
   class=MaterialExpressionScalarParameter
   properties {defaultValue=0.4, group=MuzzleSwish, parameterName=MS_AlphaProgressCeilBias, sliderMax=1}
  }
  MaterialExpressionScalarParameter_19 {class=MaterialExpressionScalarParameter, properties {group=MuzzleSwish, parameterName=MS_BaseColorAccent_Sharpness, sliderMax=1}}
  MaterialExpressionScalarParameter_2 {class=MaterialExpressionScalarParameter, properties {defaultValue=128, group=MuzzleSwish, parameterName=Scale}}
  MaterialExpressionScalarParameter_20 {class=MaterialExpressionScalarParameter, properties {group=MuzzleSwish, parameterName=MS_BaseColorAccent_FresnelAmt, sliderMax=1}}
  MaterialExpressionScalarParameter_3 {class=MaterialExpressionScalarParameter, properties {defaultValue=1, group=MuzzleSwish, parameterName=MS_EmissiveProgress, sliderMax=1}}
  MaterialExpressionScalarParameter_4 {class=MaterialExpressionScalarParameter, properties {group=MuzzleSwish, parameterName=MS_AlphaProgress, sliderMax=0.8, sliderMin=-0.1}}
  MaterialExpressionScalarParameter_5 {class=MaterialExpressionScalarParameter, properties {defaultValue=-0.3, group=MuzzleSwish, parameterName=MS_NoisePanningSpeed}}
  MaterialExpressionScalarParameter_6 {class=MaterialExpressionScalarParameter, properties {defaultValue=3, parameterName=MS_WPOAppearStartDistance}}
  MaterialExpressionScalarParameter_7 {class=MaterialExpressionScalarParameter, properties {defaultValue=8, group=MuzzleSwish, parameterName=MS_OutlineExp}}
  MaterialExpressionScalarParameter_8 {class=MaterialExpressionScalarParameter, properties {parameterName=AppearIndex DEBUG, sliderMax=1}}
  MaterialExpressionScalarParameter_9 {class=MaterialExpressionScalarParameter, properties {group=MuzzleSwish, parameterName=BodyID, sliderMax=2}}
  MaterialExpressionSetMaterialAttributes_0 {
   class=MaterialExpressionSetMaterialAttributes
   inputs [
    MaterialAttributes {from=MaterialExpressionGetMaterialAttributes_0, output=MaterialAttributes}
    Emissive Color {from=MaterialExpressionNamedRerouteUsage_0, output=""}
    Base Color {from=MaterialExpressionNamedRerouteUsage_3, output=""}
    Opacity Mask {from=MaterialExpressionNamedRerouteUsage_4, output=""}
    World Position Offset {from=MaterialExpressionNamedRerouteUsage_8, output=""}
    Opacity {from=MaterialExpressionNamedRerouteUsage_33, output=""}
   ]
   properties {attributeSetTypes [
     #1
     #2
     #5
     #3
     #4
    ]}
  }
  MaterialExpressionSmoothStep_0 {
   class=MaterialExpressionSmoothStep
   inputs [
    Min {from="", output=""}
    Max {from="", output=""}
    Value {from=MaterialExpressionSaturate_4, output=""}
   ]
   properties {}
  }
  MaterialExpressionSubtract_1 {
   class=MaterialExpressionSubtract
   inputs [
    A {from=MaterialExpressionMultiply_6}
    B {from=MaterialExpressionMultiply_5}
   ]
  }
  MaterialExpressionSubtract_11 {
   class=MaterialExpressionSubtract
   inputs [
    A {from=MaterialExpressionNamedRerouteUsage_27}
    B {from=""}
   ]
   properties {constB=0.1}
  }
  MaterialExpressionSubtract_14 {
   class=MaterialExpressionSubtract
   inputs [
    A {from=MaterialExpressionSubtract_11}
    B {from=MaterialExpressionNamedRerouteUsage_30}
   ]
  }
  MaterialExpressionSwitch_0 {
   class=MaterialExpressionSwitch
   inputs [
    SwitchValue {from=MaterialExpressionRound_0, output=""}
    Default {from="", output=""}
    0 {from=MaterialExpressionConstant_2, output=""}
    0.1 {from=MaterialExpressionConstant_2, output=""}
    0.2 {from=MaterialExpressionConstant_3, output=""}
    0.3 {from=MaterialExpressionConstant_4, output=""}
    0.4 {from=MaterialExpressionConstant_7, output=""}
    0.5 {from=MaterialExpressionConstant_8, output=""}
    0.6 {from=MaterialExpressionConstant_9, output=""}
    0.7 {from=MaterialExpressionConstant_10, output=""}
    0.8 {from=MaterialExpressionConstant_11, output=""}
    0.9 {from=MaterialExpressionConstant_12, output=""}
   ]
   properties {inputs [
     {
      inputName=0
      input {
       expression=&self:MaterialExpressionConstant_2
       inputName=None
       mask=0
       maskA=0
       maskB=0
       maskG=0
       maskR=0
       outputIndex=0
      }
     }
     {
      inputName=0.1
      input {
       expression=&self:MaterialExpressionConstant_2
       inputName=None
       mask=0
       maskA=0
       maskB=0
       maskG=0
       maskR=0
       outputIndex=0
      }
     }
     {
      inputName=0.2
      input {
       expression=&self:MaterialExpressionConstant_3
       inputName=None
       mask=0
       maskA=0
       maskB=0
       maskG=0
       maskR=0
       outputIndex=0
      }
     }
     {
      inputName=0.3
      input {
       expression=&self:MaterialExpressionConstant_4
       inputName=None
       mask=0
       maskA=0
       maskB=0
       maskG=0
       maskR=0
       outputIndex=0
      }
     }
     {
      inputName=0.4
      input {
       expression=&self:MaterialExpressionConstant_7
       inputName=None
       mask=0
       maskA=0
       maskB=0
       maskG=0
       maskR=0
       outputIndex=0
      }
     }
     {
      inputName=0.5
      input {
       expression=&self:MaterialExpressionConstant_8
       inputName=None
       mask=0
       maskA=0
       maskB=0
       maskG=0
       maskR=0
       outputIndex=0
      }
     }
     {
      inputName=0.6
      input {
       expression=&self:MaterialExpressionConstant_9
       inputName=None
       mask=0
       maskA=0
       maskB=0
       maskG=0
       maskR=0
       outputIndex=0
      }
     }
     {
      inputName=0.7
      input {
       expression=&self:MaterialExpressionConstant_10
       inputName=None
       mask=0
       maskA=0
       maskB=0
       maskG=0
       maskR=0
       outputIndex=0
      }
     }
     {
      inputName=0.8
      input {
       expression=&self:MaterialExpressionConstant_11
       inputName=None
       mask=0
       maskA=0
       maskB=0
       maskG=0
       maskR=0
       outputIndex=0
      }
     }
     {
      inputName=0.9
      input {
       expression=&self:MaterialExpressionConstant_12
       inputName=None
       mask=0
       maskA=0
       maskB=0
       maskG=0
       maskR=0
       outputIndex=0
      }
     }
    ]}
  }
  MaterialExpressionSwitch_1 {
   class=MaterialExpressionSwitch
   inputs [
    SwitchValue {from=MaterialExpressionScalarParameter_10, output=""}
    Default {from=MaterialExpressionVectorParameter_0, output=RGB}
    None {from=MaterialExpressionVectorParameter_0, output=RGB}
    None {from=MaterialExpressionVectorParameter_5, output=RGB}
    None {from=MaterialExpressionVectorParameter_6, output=RGB}
   ]
   properties {inputs [
     {
      inputName=None
      input {
       expression=&self:MaterialExpressionVectorParameter_0
       inputName=None
       mask=1
       maskA=0
       maskB=1
       maskG=1
       maskR=1
       outputIndex=0
      }
     }
     {
      inputName=None
      input {
       expression=&self:MaterialExpressionVectorParameter_5
       inputName=None
       mask=1
       maskA=0
       maskB=1
       maskG=1
       maskR=1
       outputIndex=0
      }
     }
     {
      inputName=None
      input {
       expression=&self:MaterialExpressionVectorParameter_6
       inputName=None
       mask=1
       maskA=0
       maskB=1
       maskG=1
       maskR=1
       outputIndex=0
      }
     }
    ]}
  }
  MaterialExpressionSwitch_2 {
   class=MaterialExpressionSwitch
   inputs [
    SwitchValue {from=MaterialExpressionScalarParameter_11, output=""}
    Default {from=MaterialExpressionVectorParameter_2, output=RGB}
    None {from=MaterialExpressionVectorParameter_2, output=RGB}
    None {from=MaterialExpressionVectorParameter_7, output=RGB}
    None {from=MaterialExpressionVectorParameter_8, output=RGB}
   ]
   properties {inputs [
     {
      inputName=None
      input {
       expression=&self:MaterialExpressionVectorParameter_2
       inputName=None
       mask=1
       maskA=0
       maskB=1
       maskG=1
       maskR=1
       outputIndex=0
      }
     }
     {
      inputName=None
      input {
       expression=&self:MaterialExpressionVectorParameter_7
       inputName=None
       mask=1
       maskA=0
       maskB=1
       maskG=1
       maskR=1
       outputIndex=0
      }
     }
     {
      inputName=None
      input {
       expression=&self:MaterialExpressionVectorParameter_8
       inputName=None
       mask=1
       maskA=0
       maskB=1
       maskG=1
       maskR=1
       outputIndex=0
      }
     }
    ]}
  }
  MaterialExpressionSwitch_3 {
   class=MaterialExpressionSwitch
   inputs [
    SwitchValue {from=MaterialExpressionScalarParameter_9, output=""}
    Default {from=MaterialExpressionVectorParameter_3, output=RGB}
    None {from=MaterialExpressionVectorParameter_3, output=RGB}
    None {from=MaterialExpressionVectorParameter_9, output=RGB}
    None {from=MaterialExpressionVectorParameter_10, output=RGB}
   ]
   properties {inputs [
     {
      inputName=None
      input {
       expression=&self:MaterialExpressionVectorParameter_3
       inputName=None
       mask=1
       maskA=0
       maskB=1
       maskG=1
       maskR=1
       outputIndex=0
      }
     }
     {
      inputName=None
      input {
       expression=&self:MaterialExpressionVectorParameter_9
       inputName=None
       mask=1
       maskA=0
       maskB=1
       maskG=1
       maskR=1
       outputIndex=0
      }
     }
     {
      inputName=None
      input {
       expression=&self:MaterialExpressionVectorParameter_10
       inputName=None
       mask=1
       maskA=0
       maskB=1
       maskG=1
       maskR=1
       outputIndex=0
      }
     }
    ]}
  }
  MaterialExpressionSwitch_4 {
   class=MaterialExpressionSwitch
   inputs [
    SwitchValue {from=MaterialExpressionScalarParameter_12, output=""}
    Default {from=MaterialExpressionVectorParameter_1, output=RGB}
    None {from=MaterialExpressionVectorParameter_1, output=RGB}
    None {from=MaterialExpressionVectorParameter_11, output=RGB}
    None {from=MaterialExpressionVectorParameter_12, output=RGB}
   ]
   properties {inputs [
     {
      inputName=None
      input {
       expression=&self:MaterialExpressionVectorParameter_1
       inputName=None
       mask=1
       maskA=0
       maskB=1
       maskG=1
       maskR=1
       outputIndex=0
      }
     }
     {
      inputName=None
      input {
       expression=&self:MaterialExpressionVectorParameter_11
       inputName=None
       mask=1
       maskA=0
       maskB=1
       maskG=1
       maskR=1
       outputIndex=0
      }
     }
     {
      inputName=None
      input {
       expression=&self:MaterialExpressionVectorParameter_12
       inputName=None
       mask=1
       maskA=0
       maskB=1
       maskG=1
       maskR=1
       outputIndex=0
      }
     }
    ]}
  }
  MaterialExpressionTextureCoordinate_0 {class=MaterialExpressionTextureCoordinate}
  MaterialExpressionTextureSample_0 {
   class=MaterialExpressionTextureSample
   inputs [
    UVs {from=MaterialExpressionPanner_0}
    Tex {from=""}
    "DDX(UVs)" {from=MaterialExpressionDDX_0}
    "DDY(UVs)" {from=MaterialExpressionDDY_0}
    Apply View MipBias {from=""}
   ]
   properties {mipValueMode=TMVM_Derivative, samplerType=SAMPLERTYPE_Grayscale, texture=/Game/Global/Textures/Default/Noise/LowResBlurredNoiseNormalized}
  }
  MaterialExpressionVectorParameter_0 {class=MaterialExpressionVectorParameter, properties {group=MuzzleSwish, parameterName=MS_FresnelColor0, defaultValue=<rgba 4 0 1.6215 1>}}
  MaterialExpressionVectorParameter_1 {
   class=MaterialExpressionVectorParameter
   properties {group=MuzzleSwish, parameterName=MS_OutlineColor0, defaultValue=<rgba 0.5104 1.0837 2 1>}
  }
  MaterialExpressionVectorParameter_10 {
   class=MaterialExpressionVectorParameter
   properties {group=MuzzleSwish, parameterName=MS_BaseColorTint2, defaultValue=<rgba 0.7078 0.5762 1 1>}
  }
  MaterialExpressionVectorParameter_11 {
   class=MaterialExpressionVectorParameter
   properties {group=MuzzleSwish, parameterName=MS_OutlineColor1, defaultValue=<rgba 0.5104 1.0837 2 1>}
  }
  MaterialExpressionVectorParameter_12 {
   class=MaterialExpressionVectorParameter
   properties {group=MuzzleSwish, parameterName=MS_OutlineColor2, defaultValue=<rgba 0.5104 1.0837 2 1>}
  }
  MaterialExpressionVectorParameter_2 {class=MaterialExpressionVectorParameter, properties {group=MuzzleSwish, parameterName=MS_BaseColorAccent0, defaultValue=<rgba 0.9475 0 1 1>}}
  MaterialExpressionVectorParameter_3 {class=MaterialExpressionVectorParameter, properties {group=MuzzleSwish, parameterName=MS_BaseColorTint, defaultValue=<rgba 1 1 1 1>}}
  MaterialExpressionVectorParameter_4 {
   class=MaterialExpressionVectorParameter
   properties {group=MuzzleSwish, parameterName=MS_NightTint, defaultValue=<rgba 0.0523 0.0471 0.0885 1>}
  }
  MaterialExpressionVectorParameter_5 {class=MaterialExpressionVectorParameter, properties {group=MuzzleSwish, parameterName=MS_FresnelColor1, defaultValue=<rgba 3.75 3.837 4 1>}}
  MaterialExpressionVectorParameter_6 {class=MaterialExpressionVectorParameter, properties {group=MuzzleSwish, parameterName=MS_FresnelColor2, defaultValue=<rgba 0 0.371 4 1>}}
  MaterialExpressionVectorParameter_7 {
   class=MaterialExpressionVectorParameter
   properties {group=MuzzleSwish, parameterName=MS_BaseColorAccent1, defaultValue=<rgba 0.3646 0.3646 0.3646 1>}
  }
  MaterialExpressionVectorParameter_8 {class=MaterialExpressionVectorParameter, properties {group=MuzzleSwish, parameterName=MS_BaseColorAccent2, defaultValue=<rgba 0 0.1933 1 1>}}
  MaterialExpressionVectorParameter_9 {
   class=MaterialExpressionVectorParameter
   properties {group=MuzzleSwish, parameterName=MS_BaseColorTint1, defaultValue=<rgba 0.0365 0.0365 0.0365 1>}
  }
  MaterialExpressionVertexColor_0 {class=MaterialExpressionVertexColor, inputs [], properties {}}
  MaterialExpressionVertexColor_1 {class=MaterialExpressionVertexColor, inputs [], properties {}}
  MaterialExpressionVertexNormalWS_0 {class=MaterialExpressionVertexNormalWS, inputs [], properties {}}
 }
 properties {
  bAutomaticallySetUsageInEditor=false
  bUseMaterialAttributes=true
  bUsedWithMeshDeformer=true
  bUsedWithMorphTargets=true
  bUsedWithSkeletalMesh=true
  bUsedWithStaticLighting=true
  blendMode=BLEND_Masked
  shadingModel=MSM_FromMaterialExpression
  translucencyLightingMode=TLM_VolumetricPerVertexDirectional
  usedShadingModels=Default Lit | Subsurface | Clear Coat | Subsurface Profile | Cloth
 }
}
